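# Translucent unlit deferred decal material that draws burnt-tires graffiti, judging by its name and texture, with a smoky, broken-up opacity.
# The graffiti RGB passes through a time-of-day color function and is multiplied by itself and the opacity mask.
# The mask sums two offset smoke-tile samples, applies a power curve and multiplies by the graffiti alpha through a channel-mask parameter.
# Decal lifetime opacity, scaled by ErodeStrength, is subtracted from the mask before an opacity multiplier and saturate.
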Material /Game/Environments/Helios/Sets/BurntTires/Materials/MD_BurntTires_Graffiti_A
graph {
 properties {
  bAutomaticallySetUsageInEditor=false
  blendMode=BLEND_Translucent
  materialDomain=MD_DeferredDecal
  pixelDepthOffsetMode=PDOM_Legacy
  shadingModel=MSM_Unlit
  usedShadingModels=Unlit
 }
 nodes [
  Add()
  n1 = TextureSample() {texture=/Game/Environments/Helios/Sets/BurntTires/Textures/Graffiti/T_BurntTires_Graffiti_A}
  n2 = MaterialFunctionCall(Color_Morning=@n1.RGB, Color_Day=@n1.RGB, Color_Evening=@n1.RGB, Color_Night=@n1.RGB) {materialFunction=&Color_Time_of_Day}
  n3 = TextureSample(UVs=Multiply(A=TextureCoordinate()) {constB=0.5}) {texture=&T_Inky_Smoke_Tile}
  n4 = TextureCoordinate()
  n5 = TextureSample(UVs=Add(A=Multiply(A=@n4) {constB=0.5}, B=Constant2Vector() {r=0.5})) {texture=&T_Inky_Smoke_Tile}
  n6 = Multiply(
   A=Power(Base=Add(A=@n3.R, B=@n5.R)) {constExponent=0.25}
   B=ChannelMaskParameter(None=TextureSample() {texture=&T_BurntTires_Graffiti_A}.A) {maskChannel=Green, parameterName=Decal Opacity Mask Channel, defaultValue=<rgba 0 1 0 0>}.R
  )
  Multiply(A=@n2.Result, B=Multiply(A=@n2.Result, B=@n6))
  n7 = Multiply(A=@n6, B=ScalarParameter() {defaultValue=1})
  n8 = Multiply(A=OneMinus(None=DecalLifetimeOpacity().Opacity), B=ScalarParameter() {parameterName=ErodeStrength})
  Saturate(None=Multiply(A=Subtract(A=@n7, B=@n8), B=ScalarParameter() {defaultValue=1, parameterName=Opacity Multiplier}))
  ScalarParameter() {parameterName=Metallic}
  ScalarParameter() {defaultValue=1, parameterName=Specular}
  ScalarParameter() {defaultValue=0.5, parameterName=Roughness}
  TextureSampleParameter2D() {parameterName=Decal Normal, samplerType=SAMPLERTYPE_Normal, texture=/Engine/EngineMaterials/FlatNormal}
 ]
}
